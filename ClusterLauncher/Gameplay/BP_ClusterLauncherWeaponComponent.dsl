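; Weapon component for the Cluster Launcher that signals when its secondary (apparently laser) action is allowed.
; Secondary press sends the AllowSecondary gameplay event to the instigator pawn immediately; release resends it after a short timer, if the weapon is valid.
; Unequip clears that timer and applies the laser AllowSecondary GameplayEffect directly to the pawn's ability system; EndPlay also clears it.
; The event tag is spelled "ClusterLaucher", missing an "n".
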(event EventEndPlay
  (params (EndPlayReason :type EEndPlayReason%20Enum :default "Destroyed"))
  (Utilities|Time|ClearandInvalidateTimerbyHandle (Variables|Default|GetAllowSecondaryTimerHandle)))

(event EventBP_OnSecondaryPressed
  (params (Weapon :type Fort%20Weapon%20Object%20Reference))
  (bind _asfort_player_pawn (Utilities|Casting|CastToFortPlayerPawn (Game|GetInstigator Weapon)))
  (Ability|SendGameplayEventtoActor _asfort_player_pawn "(TagName=\"Gameplay.Event.ClusterLaucher.AllowSecondary\")"))

(event EventBP_OnSecondaryReleased
  (params (Weapon :type Fort%20Weapon%20Object%20Reference))
  (bind _returnvalue (Utilities|Time|SetTimerbyEvent (delegate Custom|AllowSecondary) 0.2))
  (Variables|Default|SetAllowSecondaryTimerHandle _returnvalue))

(event Custom|AllowSecondary
  (bind _returnvalue (CallFunction|GetComponentWeapon))
  (Utilities|IsValid _returnvalue
    (:Is%20Valid
      (bind _asfort_player_pawn (Utilities|Casting|CastToFortPlayerPawn (Game|GetInstigator _returnvalue)))
      (Ability|SendGameplayEventtoActor _asfort_player_pawn "(TagName=\"Gameplay.Event.ClusterLaucher.AllowSecondary\")"))))

(event EventBP_OnUnEquipped
  (params (Weapon :type Fort%20Weapon%20Object%20Reference))
  (Utilities|FlowControl|Sequence
    (:then_0
      (Utilities|Time|ClearandInvalidateTimerbyHandle (Variables|Default|GetAllowSecondaryTimerHandle)))
    (:then_1
      (bind _asfort_player_pawn (Utilities|Casting|CastToFortPlayerPawn (Game|GetInstigator Weapon)))
      (GameplayEffects|ApplyGameplayEffectToSelf :self (Ability|GetAbilitySystemComponent _asfort_player_pawn) :GameplayEffectClass "/ClusterLauncher/Gameplay/GE_TidalCrane_ClusterLauncher_Laser_AllowSecondary.GE_TidalCrane_ClusterLauncher_Laser_AllowSecondary_C" :Level 0.0))))
