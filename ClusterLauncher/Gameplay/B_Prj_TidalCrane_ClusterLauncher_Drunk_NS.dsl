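; Cluster Launcher rocket projectile with drunk homing flight, exploding on impact or when its fuse timer expires.
; Construction script sets pre- and post-lock-on drunk movement configs before homing data replicates.
; Server timers randomize lock-on offset noise; water interaction can start delayed.
; It deregisters from the laser weapon on stop, end play or predicted-projectile swap, while handling trail, spin, loop audio and simulated-proxy explosion cues.
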(event EventEndPlay
  (params (EndPlayReason :type EEndPlayReason%20Enum :default "Destroyed"))
  ;[1 1 1] End Play
  (CallFunction|TryRemoveProjectilefromWeapon)
  (CallFunction|StopinAirLoop)
  (Utilities|Time|ClearandInvalidateTimerbyHandle (Variables|Default|GetFuseMaxTimerHandle))
  (Utilities|Time|ClearandInvalidateTimerbyHandle (Variables|Default|GetWaterInteractionDelayTimerHandle)))

(event Custom|Start%20Lock%20On%20Noise%20Update
  ;[1 1 1] Lock On Noise
  (Networking|SwitchHasAuthority
    (:Authority
      (if (not (Utilities|Time|IsValidTimerHandle (Variables|Default|GetLock%20On%20Offset%20Noise%20Timer%20Handle)))
        (CallFunction|UpdateLockonNoiseVector)))))

(event Custom|Update%20Lock%20On%20Noise%20Vector
  ;[1 1 1] Lock On Noise
  (Networking|SwitchHasAuthority
    (:Authority
      (Variables|Default|SetLock%20on%20Offset%20Noise (* (Math|Random|RandomUnitVector) (ScalableFloat|GetCachedScalableFloatValue (Variables|Default|GetRow%20Lock%20On%20Noise%20Multiplier))))
      (Utilities|Time|ClearandInvalidateTimerbyHandle (Variables|Default|GetLock%20On%20Offset%20Noise%20Timer%20Handle))
      (bind _returnvalue (Utilities|Time|SetTimerbyEvent (EventDispatchers|CreateEvent :!event Custom|Update%20Lock%20On%20Noise%20Vector) (Math|Random|RandomFloatinRange (ScalableFloat|GetCachedScalableFloatValue (Variables|Default|GetRow%20Lock%20On%20Timer%20Min)) (ScalableFloat|GetCachedScalableFloatValue (Variables|Default|GetRow%20Lock%20On%20Timer%20Max)))))
      (Variables|Default|SetLock%20On%20Offset%20Noise%20Timer%20Handle _returnvalue))))

(event EventBeginPlay
  (bind _ns_clusterlauncher_rockettrail (Variables|Default|GetNS_ClusterLauncher_RocketTrail))
  ;[1 1 1] Begin Play
  (Variables|Default|SetReached%20destination false)
  (bind _returnvalue (Utilities|Time|SetTimerbyFunctionName self "FuseTimerMax" (ScalableFloat|GetCachedScalableFloatValue (Variables|Default|GetRow%20LifeTimeMax))))
  (Variables|Default|SetFuseMaxTimerHandle _returnvalue)
  (Utilities|FlowControl|Sequence
    (:then_0
      (if (> (ScalableFloat|GetCachedScalableFloatValue (Variables|Default|GetRow%20Water%20Interaction%20Delay)) 0)
        (bind _returnvalue_2 (Utilities|Time|SetTimerbyEvent (EventDispatchers|CreateEvent :!event Custom|EnableWaterInteraction) 0.66))
        (Variables|Default|SetWaterInteractionDelayTimerHandle _returnvalue_2)
        (else
          (Variables|Projectile|SetTouchWaterBehavior "StopIfStopSimulatingOnHit"))))
    (:then_1
      (Networking|CanExecuteCosmeticEvents
        (:True
          (Components|Activation|Activate _ns_clusterlauncher_rockettrail false)
          (Rendering|SetRenderCustomDepth _ns_clusterlauncher_rockettrail true)
          (Niagara|SetNiagaraVariable(Float) _ns_clusterlauncher_rockettrail "ProjLifetime" (ScalableFloat|GetCachedScalableFloatValue (Variables|Default|GetRow%20LifeTimeMax)))
          (CallFunction|PlayinAirLoop))))))

(event Custom|Cancel%20Lock%20On%20Noise%20Update
  ;[1 1 1] Lock On Noise
  (Networking|SwitchHasAuthority
    (:Authority
      (Utilities|Time|ClearandInvalidateTimerbyHandle (Variables|Default|GetLock%20On%20Offset%20Noise%20Timer%20Handle)))))

(event Custom|PlayInAirLoop
  (bind _returnvalue (Projectile|GetInstigatorPlayerController))
  ; Audio Logic
  (Networking|CanExecuteCosmeticEvents
    (:True
      (Variables|Default|GetInAirAudioLoop
        (:then)
        (:else
          (bind _returnvalue_2 (Audio|SpawnSoundAttached "/ClusterLauncher/Assets/Audio/MSS_TidalCrane_ClusterLauncher_Projectile_Loop.MSS_TidalCrane_ClusterLauncher_Projectile_Loop" (Variables|Collision|GetCapsuleComponent)))
          (bind _output_get (Variables|Default|SetInAirAudioLoop _returnvalue_2))
          (Utilities|IsValid _returnvalue
            (:Is%20Valid
              (Audio|Parameter|SetIntegerParameter _output_get "IsEnemy" (Utilities|Select :Index (Team|GetTeamAffiliationforActortoLocalPlayer _returnvalue) :Friendly 0 :Neutral 0 :Hostile 1) :!class /Script/AudioExtensions.AudioParameterControllerInterface))))))))

(event Custom|StopInAirLoop
  ; Audio Logic
  (bind _inairaudioloop (Variables|Default|GetInAirAudioLoop)
    (:then
      (Audio|Components|Audio|FadeOut _inairaudioloop 0.5 0.0 "Linear" :!class /Script/Engine.AudioComponent))
    (:else)))

(event EventTick
  (params (DeltaSeconds :type Float%20%28single-precision%29))
  ;[1 1 1] On Tick, spin the rocket and particle trail
  (Variables|Default|SetTick%20Delta DeltaSeconds)
  (Utilities|FlowControl|Sequence
    (:then_1
      (Networking|CanExecuteCosmeticEvents
        (:True
          (if (Components|Activation|IsActive (Variables|Default|GetNS_ClusterLauncher_RocketTrail))
            (Variables|Effects|SetRocketSpin (Math|Rotator|MakeRotator 0.0 (* (Variables|Default|GetTick%20Delta) (Variables|Effects|GetRocketSpinRate))))
            (Transformation|AddLocalRotation (Variables|Default|GetSM_OstrichRocket) (Variables|Effects|GetRocketSpin) false false)))))))

(event Custom|EnableWaterInteraction
  ; EnableWaterInteraction
  (if (and (Utilities|IsValid (Variables|/Script/Engine.MovementComponent|GetUpdatedComponent (Variables|Movement|GetProjectileMovementComponent)) :!class /Script/Engine.KismetSystemLibrary) (not (Delivery|IsBeingKilled)))
    (Variables|Projectile|SetTouchWaterBehavior "StopIfStopSimulatingOnHit")
    (if (Water|IsinWater (Variables|Water|GetWaterInteractionComponent))
      (CallFunction|FuseTimerMax))))

(event ClusterLauncher|EventOnPredictedProjectileReplaced
  (params (PredictedProjectile :type Fort%20Projectile%20Base%20Object%20Reference))
  ; On Predicted Projectile Replaced
  ;
  ; Cluster Launcher projectiles start out as client-predicted, and then switch to server-auth once the "real" projectiles spawn on the client. This event is called when the projectile swap takes place.
  (bind _returnvalue (CallFunction|GetOwnerWeapon))
  (bind _asb_tidal_crane_cluster_launcher_laser (Utilities|Casting|CastToB_TidalCrane_ClusterLauncher_Laser _returnvalue))
  (bind _asb_prj_tidal_crane_cluster_launcher_drunk_ns (Utilities|Casting|CastToB_Prj_TidalCrane_ClusterLauncher_Drunk_NS PredictedProjectile))
  (Class|BTidalCraneClusterLauncherLaser|RemoveFiredProjectile _asb_tidal_crane_cluster_launcher_laser _asb_prj_tidal_crane_cluster_launcher_drunk_ns))

(event EventOnStop
  (params (Hit :type Hit%20Result%20Structure%20%28by%20ref%29))
  (bind (_bblockinghit _binitialoverlap _time _distance _location _impactpoint _normal _impactnormal _physmat _hitactor) (Collision|BreakHitResult Hit))
  ;[1 1 1] Impact and Explosion Stuff
  (CallFunction|TryRemoveProjectilefromWeapon)
  (Utilities|FlowControl|Sequence
    (:then_0
      ; Check if hit water
      (Utilities|Casting|CastToFortWaterBodyActor _hitactor
        (:then
          _)
        (:CastFailed
          (Utilities|Casting|CastToFortAthenaWater _hitactor)))
      (Variables|Default|SetHitWater true))
    (:then_1
      (Variables|Default|SetImpact%20Normal _impactnormal)
      (Variables|Effects|SetDecalLocation _impactpoint)
      (Variables|Effects|SetStopLocation (+ _impactpoint (* _impactnormal (Variables|Effects|GetExplosion%20Impact%20Offset))))
      (Variables|Effects|SetStopRotZ (Math|Rotator|MakeRotfromZ _impactnormal))
      (Utilities|IsValid _physmat
        (:Is%20Valid
          (Variables|Default|SetSurfaceType (Variables|/Script/PhysicsCore.PhysicalMaterial|GetSurfaceType _physmat)))
        (:Is%20Not%20Valid
          (Variables|Default|SetSurfaceType "SurfaceType_Default")))
      (if (not (Delivery|IsBeingKilled))
        (Delivery|DoExplosion)))))

(event Custom|FuseTimerMax
  ;[1 1 1] Impact and Explosion Stuff
  (Variables|Effects|SetTimerMaxReached? true)
  (Variables|Effects|SetStopLocation (Transformation|GetWorldLocation (Variables|Default|GetSM_OstrichRocket)))
  (if (not (Delivery|IsBeingKilled))
    (Delivery|DoExplosion)))

(event OnEnterWaterEvent(WaterInteractionComponent)
  (params (WaterBody :type Fort%20Water%20Body%20Actor%20Object%20Reference) (WaterInteractionComponent :type Fort%20Water%20Interaction%20Component%20Object%20Reference) (bIsFirstBody :type Boolean))
  ;[1 1 1] Impact and Explosion Stuff
  (Variables|Default|SetHitWater true))

(event EventOnExploded
  (params (HitActors :type Array%20of%20Actor%20Object%20References) (HitResults :type Array%20of%20Hit%20Result%20Structures))
  ;[1 1 1] Impact and Explosion Stuff
  (switch Utilities|FlowControl|Switch|SwitchonENetRole (Networking|GetLocalRole)
    (:ROLE_SimulatedProxy
      (Development|SetHiddeninGame (Variables|Default|GetSM_OstrichRocket) true false)
      ;;[1 1 1] SingleLineCheckForWorldObjectsSurfaceType
      ;;[1 0.15 0.15 0.5] Leaving this here for when we want to separate out water.
      (if (not (Variables|Default|GetHitWater))
        (GameplayCue|ExecuteGameplayCue%28Burst%29Local self (Variables|Default|GetTag%20GC%20Explosion) (Ability|GameplayCue|MakeGameplayCueParameters :NormalizedMagnitude 0.0 :RawMagnitude 0.0 :Location "0,0,0" :Normal (Variables|Default|GetImpact%20Normal) :GameplayEffectLevel 1 :AbilityLevel 1 :bReplicateLocationWhenUsingMinimalRepProxy false))
        (else
          (GameplayCue|ExecuteGameplayCue%28Burst%29Local self (Variables|Default|GetGC%20Tag%20Hit%20Water) (Ability|GameplayCue|MakeGameplayCueParameters :NormalizedMagnitude 0.0 :RawMagnitude 0.0 :Location "0,0,0" :Normal (Variables|Default|GetImpact%20Normal) :GameplayEffectLevel 1 :AbilityLevel 1 :bReplicateLocationWhenUsingMinimalRepProxy false))))
      (Rendering|SetVisibility (Variables|Default|GetSM_OstrichRocket) false false)
      ;;[1 1 1] Shake that camera
      (GameplayCue|ExecuteGameplayCue%28Burst%29Local self (Variables|Effects|GetFeedbackCue) (Ability|GameplayCue|MakeGameplayCueParameters 0.0)) ; Cam shake + controller vibration
      (CallFunction|StopinAirLoop))))

(event EventDestroyed
  ;[1 1 1] Impact and Explosion Stuff
  (switch Utilities|FlowControl|Switch|SwitchonENetRole (Networking|GetLocalRole)
    (:ROLE_SimulatedProxy
      (Components|Activation|SetActive (Variables|Default|GetNS_ClusterLauncher_RocketTrail) false false)
      (CallFunction|StopinAirLoop))))

(event OnProjectileStop(ProjectileMovementComponent)
  (params (ImpactResult :type Hit%20Result%20Structure%20%28by%20ref%29))
  ;[1 1 1] Impact and Explosion Stuff
  (CallFunction|StopinAirLoop)
  (Components|Activation|Deactivate (Variables|Default|GetNS_ClusterLauncher_RocketTrail)))

(fn GetLockOnOffset
  (return (Variables|Default|GetLock%20on%20Offset%20Noise)))

(fn InitializeDrunkConfig
  (bind _asfort_projectile_move_comp_drunk_cluster_launcher (Utilities|Casting|CastToFortProjectileMoveComp_DrunkClusterLauncher (Variables|Movement|GetProjectileMovementComponent)))
  (Class|FortProjectileMoveCompDrunkClusterLauncher|SetClusterLauncherDrunkConfigs _asfort_projectile_move_comp_drunk_cluster_launcher (Utilities|Struct|MakeProjectileMovementDrunkConfig (Variables|Default|GetDrunk%20Speed%20Scale%20Curve) 0 (ScalableFloat|GetCachedScalableFloatValue (Variables|PreLockOnDrunk|GetRow%20Init%20Drunk%20Velocity%20Initial%20Delay)) (ScalableFloat|GetCachedScalableFloatValue (Variables|PreLockOnDrunk|GetRow%20Init%20%20Drunk%20Velocity%20Duration)) (ScalableFloat|GetCachedScalableFloatValue (Variables|PreLockOnDrunk|GetRow%20Init%20DrunkVelocityBlendOutMultiplier)) (ScalableFloat|GetCachedScalableFloatValue (Variables|PreLockOnDrunk|GetRow%20Init%20DirectionChangeRate)) (ScalableFloat|GetCachedScalableFloatValue (Variables|PreLockOnDrunk|GetRow%20Init%20MaxRandomTurnAngle)) (ScalableFloat|GetCachedScalableFloatValue (Variables|PreLockOnDrunk|GetRow%20Init%20MinRandomTurnAngleMultiplier)) (ScalableFloat|GetCachedScalableFloatValue (Variables|PreLockOnDrunk|GetRow%20Init%20DrunkTargetPositionLookAheadDist)) (ScalableFloat|GetCachedScalableFloatValue (Variables|PreLockOnDrunk|GetRow%20Init%20DrunkTargetAngleBlendOut)) (ScalableFloat|GetCachedScalableFloatValue (Variables|PreLockOnDrunk|GetRow%20Init%20HomingDrunkDuration)) (ScalableFloat|GetCachedScalableFloatValue (Variables|PreLockOnDrunk|GetRow%20Init%20HomingRandomTargetPosRadius)) (ScalableFloat|GetCachedScalableFloatValue (Variables|PreLockOnDrunk|GetRow%20Init%20HomingPassedTargetSlackTime)) (ScalableFloat|GetCachedScalableFloatValue (Variables|PreLockOnDrunk|GetRow%20Init%20HomingDrunkBlendOutRange)) (ScalableFloat|GetCachedScalableFloatValue (Variables|PreLockOnDrunk|GetRow%20Init%20TimeThresholdBlendOutTime)) (ScalableFloat|GetCachedScalableFloatValue (Variables|PreLockOnDrunk|GetRow%20Init%20DurationBlendOutTime)) (ScalableFloat|GetCachedScalableFloatValue (Variables|PreLockOnDrunk|GetRow%20Init%20HomingDrunkBlendOutTurnSpeed))) (Utilities|Struct|MakeProjectileMovementDrunkConfig (Variables|Default|GetDrunk%20Speed%20Scale%20Curve) 0 (ScalableFloat|GetCachedScalableFloatValue (Variables|PostLockOnDrunk|GetRow%20Post%20Drunk%20Velocity%20Initial%20Delay)) (ScalableFloat|GetCachedScalableFloatValue (Variables|PostLockOnDrunk|GetRow%20Post%20%20Drunk%20Velocity%20Duration)) (ScalableFloat|GetCachedScalableFloatValue (Variables|PostLockOnDrunk|GetRow%20Post%20DrunkVelocityBlendOutMultiplier)) (ScalableFloat|GetCachedScalableFloatValue (Variables|PostLockOnDrunk|GetRow%20Post%20DirectionChangeRate)) (ScalableFloat|GetCachedScalableFloatValue (Variables|PostLockOnDrunk|GetRow%20Post%20MaxRandomTurnAngle)) (ScalableFloat|GetCachedScalableFloatValue (Variables|PostLockOnDrunk|GetRow%20Post%20MinRandomTurnAngleMultiplier)) (ScalableFloat|GetCachedScalableFloatValue (Variables|PostLockOnDrunk|GetRow%20Post%20DrunkTargetPositionLookAheadDist)) (ScalableFloat|GetCachedScalableFloatValue (Variables|PostLockOnDrunk|GetRow%20Post%20DrunkTargetAngleBlendOut)) (ScalableFloat|GetCachedScalableFloatValue (Variables|PostLockOnDrunk|GetRow%20Post%20HomingDrunkDuration)) (ScalableFloat|GetCachedScalableFloatValue (Variables|PostLockOnDrunk|GetRow%20Post%20HomingRandomTargetPosRadius)) (ScalableFloat|GetCachedScalableFloatValue (Variables|PostLockOnDrunk|GetRow%20Post%20HomingPassedTargetSlackTime)) (ScalableFloat|GetCachedScalableFloatValue (Variables|PostLockOnDrunk|GetRow%20Post%20HomingDrunkBlendOutRange)) (ScalableFloat|GetCachedScalableFloatValue (Variables|PostLockOnDrunk|GetRow%20Post%20TimeThresholdBlendOutTime)) (ScalableFloat|GetCachedScalableFloatValue (Variables|PostLockOnDrunk|GetRow%20Post%20DurationBlendOutTime)) (ScalableFloat|GetCachedScalableFloatValue (Variables|PostLockOnDrunk|GetRow%20Post%20HomingDrunkBlendOutTurnSpeed)) true)))

(fn TryRemoveProjectileFromWeapon
  (bind _returnvalue (CallFunction|GetOwnerWeapon))
  (bind _asb_tidal_crane_cluster_launcher_laser (Utilities|Casting|CastToB_TidalCrane_ClusterLauncher_Laser _returnvalue))
  (Class|BTidalCraneClusterLauncherLaser|RemoveFiredProjectile _asb_tidal_crane_cluster_launcher_laser self)
  (Audio|Components|Audio|FadeOut (Variables|/ClusterLauncher/Gameplay/B_TidalCrane_ClusterLauncher_Laser.B_TidalCrane_ClusterLauncher_Laser_C|GetProjectileInAirAudio _asb_tidal_crane_cluster_launcher_laser) 0.5 0.0 "Linear" :!class /Script/Engine.AudioComponent))

(fn ConstructionScript
  ; Init Drunk Config Data in construction script so it is init before Homing Data is replicated. Drunk Configs are only used upon updating homing data. So if they are edited after homing has been set, the changes will not take effect.
  (CallFunction|InitializeDrunkConfig))
