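; Laser-guided cluster launcher weapon that steers its fired missiles and drives the targeting beam, enemy-visible glint and aim-down-sights laser audio.
; Stored projectiles home on the aim point with pattern offsets and optional noise, or spread forward when cleared.
; Autonomous proxies take local trace results and interpolate the Niagara beam end within an angle threshold.
; Simulated proxies use replicated hits and toggles, retrying beam setup until the owner replicates.
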(event EventOnSetTargeting
  (params (bNewIsTargeting :type Boolean))
  (bind _ns_laser_glint (Variables|Default|GetNS_Laser_Glint))
  (|CallParentFunction bNewIsTargeting)
  (Utilities|FlowControl|Sequence
    (:then_0
      (if (Networking|IsServer)
        (Variables|Default|SetIsAimDownSightsReplicated bNewIsTargeting)))
    (:then_1
      (if (and (not (Pawn|IsLocallyControlled (Game|GetInstigator))) (not (Team|OnLocalPlayersTeam (Game|GetInstigator))))
        (Utilities|IsValid (Variables|Default|GetNS_Laser_Glint)
          (:Is%20Valid
            (if bNewIsTargeting
              (Components|Activation|Activate _ns_laser_glint false)
              (else
                (Components|Activation|Deactivate _ns_laser_glint)))))))))

(event EventEndPlay
  (params (EndPlayReason :type EEndPlayReason%20Enum :default "Destroyed"))
  ;[1 1 1] End Play
  (CallFunction|ClearFiredProjectiles)
  (Utilities|Time|ClearandInvalidateTimerbyHandle (Variables|Default|GetRetryInitializeBeamForSimProxyOwnerTimerHandle)))

(event EventBeginPlay
  ;[1 1 1] Begin Play
  (Variables|Default|SetbIsLocal (Pawn|IsLocallyViewed (Game|GetInstigator))) ; Used for low ammo sound setup
  (bind _asfort_player_controller (Utilities|Casting|CastToFortPlayerController (Fortnite|GetControllerfromActor (Game|GetInstigator))))
  (bind _output_get (Variables|Default|SetCachedPlayerController _asfort_player_controller))
  (Variables|Default|SetPlayerPawn (Game|Player|GetPlayerPawn _output_get))
  ;; SetRearMuzzleFlash
  (Variables|Default|SetMissileCheckTargetFrequency (ScalableFloat|GetValueatLevel (Variables|Default|GetRow_MissileCheckUIFrequency)))
  (Variables|Default|SetRearMuzzleFXSystem "/ClusterLauncher/Assets/FX/NS_ClusterLauncher_MuzzleFlash.NS_ClusterLauncher_MuzzleFlash")
  (Variables|Default|SetRearMuzzleOffsetMultiplier "0, 0, 0")
  (Default|BindEventtoOnAimDownSightsChanged (EventDispatchers|CreateEvent :!event Custom|HandleADSLaserAudio)))

(event Custom|onAimDownSightsChanged_Event
  (params (AimDownSights :type Boolean)))

(event FortWeapon|EventOnEquippedAndAttached
  ;[1 1 1] OnEquippedAndAttached
  (|CallParentFunction)
  (Niagara|SetNiagaraVariable(Bool) (Variables|Default|GetNS_ClusterLauncher_Targeting) "IsLocalPlayer" (Variables|Default|GetbIsLocal))
  (Beam|InitializeBeamforSimulatedProxyOwner))

(event FortWeapon|EventOnUnEquippedAndDetached
  ;[1 1 1] OnUnEquippedAndDetached
  (|CallParentFunction)
  (Utilities|Time|ClearandInvalidateTimerbyHandle (Variables|Default|GetRetryInitializeBeamForSimProxyOwnerTimerHandle)))

(event EventTick
  (params (DeltaSeconds :type Float%20%28single-precision%29))
  ;[0.15 0.05719 0.1428 0.5] Tick: For AutoProxy, interpolate the beam impact location for smoothness.
  (|CallParentFunction DeltaSeconds)
  (Default|SwitchNetRole
    (:Autonomous%20Proxy
      (if (Components|Activation|IsActive (Variables|Default|GetNS_ClusterLauncher_Targeting))
        (Beam|InterpAutoProxyBeamEndLocation DeltaSeconds)))))

(event FortWeapon|EventOnLocalUnEquip
  ; OnLocalUnEquip
  (CallFunction|ClearFiredProjectiles))

(event Custom|HandleADSLaserAudio
  (params (AimDownSights :type Boolean))
  (bind _laseraudio (Variables|Default|GetLaserAudio))
  ; Audio
  ;; Used for AnimBP
  (bind _output_get (Variables|Default|SetIsAimDownSightsLocal AimDownSights))
  (if (Pawn|IsLocallyViewed (Variables|Default|GetPlayerPawn))
    (if _output_get
      (Audio|SpawnSoundAttached "/ClusterLauncher/Assets/Audio/MSS_TidalCrane_ClusterLauncher_Laser_Loop_Start_Preset.MSS_TidalCrane_ClusterLauncher_Laser_Loop_Start_Preset" (Variables|Visual|GetWeaponMesh))
      (Audio|Components|Audio|FadeIn _laseraudio 0.1 1.0 0.0 "Linear" :!class /Script/Engine.AudioComponent)
      (else
        (Audio|SpawnSoundAttached "/ClusterLauncher/Assets/Audio/MSS_TidalCrane_ClusterLauncher_Laser_Loop_End_Preset.MSS_TidalCrane_ClusterLauncher_Laser_Loop_End_Preset" (Variables|Visual|GetWeaponMesh))
        (Audio|Components|Audio|FadeOut _laseraudio 0.1 0.0 "Linear" :!class /Script/Engine.AudioComponent)))))

(fn ClearFiredProjectiles
  (CallFunction|ClearHomingTargetforProjectiles)
  (for _array_element (Variables|Default|GetFired%20Projectiles)
    (Class|BPrjTidalCraneClusterLauncherDrunkNS|CancelLockonNoiseUpdate _array_element))
  (Utilities|Array|Clear (Variables|Default|GetFired%20Projectiles)))

(fn ClearHomingTargetForProjectiles
  (for (_array_element _array_index) (Variables|Default|GetFired%20Projectiles)
    (Projectile|SetHomingData _array_element _array_element (Utilities|Struct|MakeProjectileHomingData "DrunkArtillery" 1.5e+02 2e+02 4.0 0.0 -1 0 0 (+ (* (Math|Rotator|GetRotationXVector (Math|Rotator|CombineRotators (* (.RelativeRotation (Utilities|Struct|BreakFortBulletPatternEntryData (Utilities|Array|Get(acopy) (Variables|Default|GetBullet%20Pattern) _array_index))) (ScalableFloat|GetCachedScalableFloatValue (Variables|Default|GetRow%20PostLaserSpreadPatternMultiplier))) (Transformation|GetWorldRotation (Variables|/Script/FortniteGame.FortProjectileBase|GetCapsuleComponent _array_element)))) (ScalableFloat|GetCachedScalableFloatValue (Variables|Default|GetRow%20Look%20Ahead%20Dist%20Clear%20Targets))) (Projectile|GetCurrentLocation _array_element))) :!class /Script/FortniteGame.FortProjectileMovementInterface)
    (Class|BPrjTidalCraneClusterLauncherDrunkNS|CancelLockonNoiseUpdate _array_element))
  (Variables|Default|SetIsHomingMissilesToTarget false))

(fn GetAimedatEnemywithLOS
  (Default|GetInstigatorPawn)
  (bind _returnvalue (Fortnite|GetAimedAtEnemy (Variables|Default|GetCachedPlayerController) (ScalableFloat|GetValueatLevel (Variables|Default|GetRow_PawnLock_HalfWidth)) (ScalableFloat|GetValueatLevel (Variables|Default|GetRow_PawnLock_HalfHeight)) (ScalableFloat|GetValueatLevel (Variables|Default|GetRow_PawnLock_Range)) true))
  (bind _output_get (LocalVariables|SetTargetEnemy _returnvalue))
  (Utilities|IsValid _output_get
    (:Is%20Valid
      (bind _returnvalue_2 (Fortnite|CheckLineOfSightToActorWithChannel (Pawn|GetPlayerViewPoint (Variables|Default|GetCachedPlayerController)) (LocalVariables|GetTargetEnemy) "ECC_GameTraceChannel7" (Variables|Default|GetCachedPlayerPawn)))
      (if _returnvalue_2
        (return (LocalVariables|GetTargetEnemy))
        (else
          _)))
    (:Is%20Not%20Valid
      _))
  (return 0))

(fn GetFiredProjectiles
  (params (Fired%20Projectiles :type Array%20of%20Fort%20Projectile%20Base%20Object%20References) (Bullet%20Pattern%20 :type Array%20of%20Fort%20Bullet%20Pattern%20Entry%20Data%20Structures))
  (Variables|Default|SetBullet%20Pattern Bullet%20Pattern%20)
  ; No need to store projectiles on sim proxy players.
  (Utilities|IsValid (Variables|Default|GetPlayerPawn)
    (:Is%20Valid
      (if (Utilities|Enum|NotEqual(Enum) (Networking|GetLocalRole (Variables|Default|GetPlayerPawn)) "ROLE_SimulatedProxy")
        (for _array_element Fired%20Projectiles
          (bind _asb_prj_tidal_crane_cluster_launcher_drunk_ns (Utilities|Casting|CastToB_Prj_TidalCrane_ClusterLauncher_Drunk_NS _array_element))
          (Utilities|Array|AddUnique (Variables|Default|GetFired%20Projectiles) _asb_prj_tidal_crane_cluster_launcher_drunk_ns))))))

(fn GetInstigatorPawn
  (Utilities|IsValid (Variables|Default|GetCachedPlayerPawn)
    (:Is%20Valid
      _)
    (:Is%20Not%20Valid
      (bind _asfort_player_pawn (Utilities|Casting|CastToFortPlayerPawn (Game|GetInstigator)))
      (Variables|Default|SetCachedPlayerPawn _asfort_player_pawn)))
  (return (Variables|Default|GetCachedPlayerPawn)))

(fn GetAnimData
  (return (FortWeapon|GetRemainingAmmo) (Variables|Default|GetIsAimDownSightsLocal) (FortWeapon|IsReloading)))

(fn GetBeamInfo
  ;[0.15 0.04684 0.1154 0.5] Returns information regarding the beam.
  ; Only makes sense to call the function when the weapon is firing.
  (return (Variables|Default|GetBeamTraceHitResult)))

(fn InitializeBeamForSimulatedProxyOwner
  ;[0.07337 0.03917 0.1406 0.5] InitializeBeamForSimulatedProxyOwner: Initialize the beam so that it can be used by the FortWeaponFxComponent.
  (Utilities|IsValid (Actor|GetOwner)
    (:Is%20Valid
      (Default|SwitchNetRole
        (:Simulated%20Proxy
          (Effects|InitializeBeamFX (Variables|Default|GetNS_ClusterLauncher_Targeting))
          (Utilities|IsValid (Actor|GetComponentbyClass "/Script/FortniteGame.FortWeaponFxComponent_Beam")
            (:Is%20Not%20Valid
              (Game|AddComponentbyClass "/Script/FortniteGame.FortWeaponFxComponent_Beam"))))
        (:Invalid
          ;;[0.4115 0 0 0.5] Warning: SwitchNetRole returned Invalid value.
          (Fortnite|FortShippingLog self "InitializeBeamForSimulatedProxyOwner: SwitchNetRole returned Invalid. Unexpected output." true))))
    (:Is%20Not%20Valid
      ;; There's a chance that when the weapon comes into relevancy, the Owner (pawn) has not yet replicated.
      ;; This is especially true when loading into a new area of the map after respawning or when starting to spectate a new player that was previously outside the relevancy range.
      ;; Attempt to initialize the beam in the next frame.
      (bind _returnvalue (Utilities|Time|SetTimerforNextTickbyEvent (EventDispatchers|CreateEvent :!event InitializeBeamForSimulatedProxyOwner)))
      (Variables|Default|SetRetryInitializeBeamForSimProxyOwnerTimerHandle _returnvalue))))

(fn InterpAutoProxyBeamEndLocation
  (params (DeltaTime :type Float%20%28single-precision%29))
  ;[0.04766 0.1021 0.15 0.5] InterpAutoProxyBeamEndLocation: For autoproxy, Interpolate the beam from the current end location to the target end location if within interpolation threshold.
  (Default|SwitchNetRole
    (:Autonomous%20Proxy
      (bind (_returnvalue _anglebetweencurrentandtargetbeam) (Beam|ShouldInterpAutoProxyBeamEndLocation (Variables|Default|GetCurrentBeamEndLocation) (Variables|Default|GetTargetBeamEndLocation)))
      (bind _output_get (Variables|Default|SetCurrentBeamEndLocation (select _returnvalue (Math|Interpolation|VInterptoConstant (Variables|Default|GetCurrentBeamEndLocation) (Variables|Default|GetTargetBeamEndLocation) DeltaTime (ScalableFloat|GetCachedScalableFloatValue (Variables|Default|GetAutoProxyBeamEndLocationInterpSpeed))) (Variables|Default|GetTargetBeamEndLocation))))
      (Niagara|SetNiagaraVariable(Position) (Variables|Default|GetNS_ClusterLauncher_Targeting) "BeamEnd_0" _output_get))))

(fn InterpSimProxyBeamEndLocation
  (params (DeltaTime :type Float%20%28single-precision%29))
  ;[0.04766 0.1021 0.15 0.5] InterpAutoProxyBeamEndLocation: For autoproxy, Interpolate the beam from the current end location to the target end location if within interpolation threshold.
  (Default|SwitchNetRole
    (:Simulated%20Proxy
      (if (Variables|Default|GetbSimProxyInterpReady)
        (bind _output_get (Variables|Default|SetCurrentBeamEndLocation (Math|Interpolation|VInterpTo (Variables|Default|GetCurrentBeamEndLocation) (Variables|Default|GetTargetBeamEndLocation) DeltaTime (ScalableFloat|GetCachedScalableFloatValue (Variables|Default|GetRow%20Sim%20Proxy%20Beam%20End%20Location%20Interp%20Speed)))))
        (Niagara|SetNiagaraVariable(Position) (Variables|Default|GetNS_ClusterLauncher_Targeting) "BeamEnd_0" _output_get)))))

(fn OnRep_BeamTraceHitResult
  (bind (_bblockinghit _binitialoverlap _time _distance _location _impactpoint _normal _impactnormal _physmat _hitactor _hitcomponent _hitbonename _bonename _hititem _elementindex _faceindex _tracestart _traceend) (Collision|BreakHitResult (Variables|Default|GetBeamTraceHitResult)))
  (Default|SwitchNetRole
    (:Simulated%20Proxy
      (Beam|RefreshBeamNiagaraImpactValue)
      (Variables|Default|SetTargetBeamEndLocation (select _bblockinghit _impactpoint _traceend))
      (if (not (Variables|Default|GetbSimProxyInterpReady))
        (bind _output_get (Variables|Default|SetCurrentBeamEndLocation (Variables|Default|GetTargetBeamEndLocation)))
        (Niagara|SetNiagaraVariable(Position) (Variables|Default|GetNS_ClusterLauncher_Targeting) "BeamEnd_0" _output_get)
        (Variables|Default|SetbSimProxyInterpReady true)))
    (:Autonomous%20Proxy
      (Beam|RefreshBeamNiagaraImpactValue))))

(fn OnRep_IsAimDownSightsReplicated
  (Variables|Default|SetIsAimDownSightsLocal (Variables|Default|GetIsAimDownSightsReplicated)))

(fn OnRep_ToggleSimProxyBeam
  (bind _togglesimproxybeam (Variables|Default|GetToggleSimProxyBeam))
  (bind _ns_clusterlauncher_targeting (Variables|Default|GetNS_ClusterLauncher_Targeting))
  (Default|SwitchNetRole
    (:Simulated%20Proxy
      (Rendering|SetVisibility (Variables|Default|GetNS_ClusterLauncher_Targeting) _togglesimproxybeam true)
      (if _togglesimproxybeam
        (Components|Activation|Activate _ns_clusterlauncher_targeting false)
        (else
          (Components|Activation|Deactivate _ns_clusterlauncher_targeting)
          (Variables|Default|SetbSimProxyInterpReady false))))))

(fn RefreshBeamNiagaraImpactValue
  (bind _beamtracehitresult (Variables|Default|GetBeamTraceHitResult))
  (bind (_bblockinghit _binitialoverlap _time _distance _location _impactpoint _normal _impactnormal _physmat) (Collision|BreakHitResult _beamtracehitresult))
  (bind _ns_clusterlauncher_targeting (Variables|Default|GetNS_ClusterLauncher_Targeting))
  ;[0.08003 0.05506 0.15 0.5] RefreshBeamNiagaraImpactValue: Run for all clients and used to adjust the niagara beam based on the latest hit result value.
  (Networking|CanExecuteCosmeticEvents
    (:True
      (Niagara|SetNiagaraVariable(Bool) _ns_clusterlauncher_targeting "IsObject" _bblockinghit)
      (Effects|Components|ParticleSystem|SetVectorParameter _ns_clusterlauncher_targeting "ImpactNormal" _impactnormal)
      (Utilities|IsValid _physmat
        (:Is%20Valid
          (Effects|Components|ParticleSystem|SetBoolParameter _ns_clusterlauncher_targeting "IsWater" (Utilities|Enum|Equal(Enum) (Variables|/Script/PhysicsCore.PhysicalMaterial|GetSurfaceType _physmat) "SurfaceType8")))
        (:Is%20Not%20Valid
          (Effects|Components|ParticleSystem|SetBoolParameter _ns_clusterlauncher_targeting "IsWater" false))))))

(fn RemoveFiredProjectile
  (params (Projectile :type B%20Prj%20Tidal%20Crane%20Cluster%20Launcher%20Drunk%20NS%20Object%20Reference))
  (bind _returnvalue (Utilities|Array|RemoveItem (Variables|Default|GetFired%20Projectiles) Projectile))
  (if _returnvalue
    (if (Math|Boolean|ANDBoolean (Utilities|Array|IsEmpty (Variables|Default|GetFired%20Projectiles)) (not (FortWeapon|IsReloading)) (Networking|IsServer) (FortWeapon|IsTargeting))
      (GameplayEffects|ApplyGameplayEffectToTarget (Ability|GetAbilitySystemComponent (Variables|Default|GetPlayerPawn)) "/ClusterLauncher/Gameplay/GE_TidalCrane_ClusterLauncher_Laser_PreventSecondary.GE_TidalCrane_ClusterLauncher_Laser_PreventSecondary_C" (Ability|GetAbilitySystemComponent (Variables|Default|GetPlayerPawn)) 0.0 (Utilities|Struct|MakeGameplayEffectContextHandle)))))

(fn SetNewHomingTarget
  (params (Aim%20Position :type Vector) (Player%20Aim%20Rotator :type Rotator))
  (bind _row_lock_on_noise_toggle (Variables|Default|GetRow%20Lock%20On%20Noise%20Toggle))
  (bind _returnvalue (ScalableFloat|GetValueasBool _row_lock_on_noise_toggle))
  (for (_array_element _array_index) (Variables|Default|GetFired%20Projectiles)
    (if _returnvalue
      (Class|BPrjTidalCraneClusterLauncherDrunkNS|StartLockonNoiseUpdate _array_element)
      (bind _lockonoffsetnoise (Class|BPrjTidalCraneClusterLauncherDrunkNS|GetLockonOffset _array_element))
      (else
        _))
    (Projectile|SetHomingData _array_element _array_element (Utilities|Struct|MakeProjectileHomingData "LockOn" (ScalableFloat|GetCachedScalableFloatValue (Variables|Default|GetRow%20Min%20Turn%20Rate)) (ScalableFloat|GetCachedScalableFloatValue (Variables|Default|GetRow%20Max%20Turn%20Rate)) (ScalableFloat|GetCachedScalableFloatValue (Variables|Default|GetRow%20Turn%20Rate%20Ramp%20Time)) 0.0 -1 0 0 (Variables|GetAim%20Position) (+ (select (and _returnvalue (> (Math|Vector|VectorLength (- (Transformation|GetActorLocation _array_element) (Variables|GetAim%20Position))) (ScalableFloat|GetCachedScalableFloatValue (Variables|Default|GetRow%20Targeting%20Noise%20Close%20Distance%20Limit)))) _lockonoffsetnoise "0, 0, 0") (Math|Vector|RotateVectorAroundAxis (Math|Vector|RotateVectorAroundAxis (Math|Vector|MakeVector 0.0 (* (.yaw (.RelativeRotation (Utilities|Struct|BreakFortBulletPatternEntryData (Utilities|Array|Get(acopy) (Variables|Default|GetBullet%20Pattern) _array_index)))) (ScalableFloat|GetCachedScalableFloatValue (Variables|Default|GetRow%20Target%20Pattern%20Offset%20Multiplier))) (* (.pitch (.RelativeRotation (Utilities|Struct|BreakFortBulletPatternEntryData (Utilities|Array|Get(acopy) (Variables|Default|GetBullet%20Pattern) _array_index)))) (ScalableFloat|GetCachedScalableFloatValue (Variables|Default|GetRow%20Target%20Pattern%20Offset%20Multiplier)))) (.yaw (Variables|GetPlayer%20Aim%20Rotator)) "0.000000,0.000000,1.000000") (.pitch (Variables|GetPlayer%20Aim%20Rotator)) (* (Math|Vector|GetRightVector (Variables|GetPlayer%20Aim%20Rotator)) -1.0))) "Absolute") :!class /Script/FortniteGame.FortProjectileMovementInterface))
  (Variables|Default|SetIsHomingMissilesToTarget true))

(fn SetAutoProxyBeamInfo
  (params (BeamHitResult :type Hit%20Result%20Structure))
  ;[0.15 0.02474 0.02474 0.5] SetAutoProxyBeamInfo: AutoProxies handle their own beam visual running local traces.
  ; Called directly from GA_TidalCrane_ClusterLauncher_BeamVisualTrace.
  (Default|SwitchNetRole
    (:Autonomous%20Proxy
      (bind _output_get (Variables|Default|SetBeamTraceHitResult BeamHitResult))
      (Beam|RefreshBeamNiagaraImpactValue)
      (bind _output_get_2 (Variables|Default|SetTargetBeamEndLocation (select (Collision|BreakHitResult _output_get) (.ImpactPoint (Collision|BreakHitResult _output_get)) (.TraceEnd (Collision|BreakHitResult _output_get)))))
      ;;[0.04262 0.06237 0.15 0.5] Instantly update the end point of the niagara effect if it is not within interpolation threshold.
      (bind _returnvalue (Beam|ShouldInterpAutoProxyBeamEndLocation (Variables|Default|GetCurrentBeamEndLocation) _output_get_2))
      (if (not _returnvalue)
        (bind _output_get_3 (Variables|Default|SetCurrentBeamEndLocation (Variables|Default|GetTargetBeamEndLocation)))
        (Niagara|SetNiagaraVariable(Position) (Variables|Default|GetNS_ClusterLauncher_Targeting) "BeamEnd_0" _output_get_3)))))

(fn SetAutoProxyBeamNiagaraEnable
  (params (bEnable :type Boolean))
  (bind _ns_clusterlauncher_targeting (Variables|Default|GetNS_ClusterLauncher_Targeting))
  ;[0.07217 0.03029 0.15 0.5] SetAutoProxyBeamNiagaraEnable: Activate or deactivate the beam as necessary.
  ; This is directly invoked from GA_TidalCrane_ClusterLaucher_BeamVisualTrace.
  (Default|SwitchNetRole
    (:Autonomous%20Proxy
      (if bEnable
        (Components|Activation|Activate _ns_clusterlauncher_targeting false)
        (else
          (Components|Activation|Deactivate _ns_clusterlauncher_targeting))))))

(fn SetBeamInfo
  (params (BeamTraceHitResult :type Hit%20Result%20Structure))
  ;[0.0511 0.2309 0.2225 0.5] SetBeamInfo: Update the beam info based on the traces run for the weapon fire.
  ; Called from GA_ThermalLance_PrimaryOnHit.
  (Default|SwitchNetRole
    (:Simulated%20Proxy
      _)
    (:Authority
      _))
  (Variables|Default|SetBeamTraceHitResult BeamTraceHitResult))

(fn SetSimProxyBeamNiagaraEnable
  (params (Enable :type Boolean))
  (Variables|Default|SetToggleSimProxyBeam Enable))

(fn ShouldInterpAutoProxyBeamEndLocation
  (params (CurrentBeamEndLocation :type Vector) (TargetBeamEndLocation :type Vector))
  (bind (_aimdir _returnvalue) (FortWeapon|GetDamageStartLocation -1))
  ;[0.02079 0.1463 0.15 0.5] ShouldInterpAutoProxyBeamEndLocation: Should the end location interp from current to target based on the angle between their directions.
  (bind _returnvalue_2 (Math|Trig|Acos(Degrees) (Math|Vector|DotProduct (Math|Vector|GetUnitDirection(Vector) _returnvalue CurrentBeamEndLocation) (Math|Vector|GetUnitDirection(Vector) _returnvalue TargetBeamEndLocation))))
  (return (< _returnvalue_2 (ScalableFloat|GetCachedScalableFloatValue (Variables|Default|GetRow%20Auto%20Proxy%20Beam%20End%20Location%20Non%20Interp%20Angle))) _returnvalue_2))

(fn ConstructionScript
  (|CallParentFunction))
