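; Gameplay ability that drives the Cluster Launcher's laser beam visual while active.
; On activation, self-rescheduling timers update the endpoint, every tick on clients and at a configurable interval on the server, and a gameplay cue is added after a net sync.
; Each update traces the beam, feeding simulated-proxy (server) or autonomous-proxy (client) Niagara, hidden when targeting near walls, or when the client isn't targeting.
; Ending clears timers and disables the beam.
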(event Ability|EventActivateAbilityFromEvent
  (params (EventData :type Gameplay%20Event%20Data%20Structure%20%28by%20ref%29))
  ;[0.01086 0.2066 0.04136 0.5] ActivateAbility: Setup delegates that are invoked when the laser weapon starts and stops firing.
  (Utilities|FlowControl|Sequence
    (:then_0
      (bind _outfortplayerpawn (CallFunction|GetAvatarasValidFortPlayerPawn)
        (:ValidFortPlayerPawn
          (bind _output_get (Variables|Transient|SetPlayerPawn _outfortplayerpawn))
          (Utilities|IsValid _output_get
            (:Is%20Valid
              (Variables|Transient|SetPlayer%20Controller (Fortnite|GetFortPlayerControllerfromActor _output_get)))))))
    (:then_1
      (CallFunction|SetWeaponAttackingEnabled true))))

(event Custom|SetWeaponAttackingEnabled
  (params (bEnable :type Boolean))
  ;[0.1478 0.04488 0.15 0.5] Invoked when the laser weapon starts and stops firing.
  (if bEnable
    (Utilities|FlowControl|Sequence
      (:then_0
        (if (Networking|IsServer)
          (bind _returnvalue (Utilities|Time|SetTimerbyEvent (EventDispatchers|CreateEvent :!event Custom|%20ServerUpdateBeamEndpoint) (ScalableFloat|GetCachedScalableFloatValue (Variables|Default|GetBeamServerUpdateInterval))))
          (Variables|Transient|SetServerUpdateBeamEndpointTimerHandle _returnvalue)
          (else
            (bind _returnvalue_2 (Utilities|Time|SetTimerforNextTickbyEvent (EventDispatchers|CreateEvent :!event Custom|ClientUpdateBeamEndpoint)))
            (Variables|Transient|SetClientUpdateBeamEndpointTimerHandle _returnvalue_2)
            (CallFunction|UpdateBeamEndpoint))))
      (:then_1
        (Ability|Tasks|WaitNetSync "OnlyClientWait")
        (Ability|AddGameplayCueToOwner (Variables|Default|GetGC_BeamVisual))))
    (else
      ;;[1 1 1] Player has stopped firing, handle tasks and niagara effect on the laser.
      (Utilities|Time|ClearandInvalidateTimerbyHandle (Variables|Transient|GetClientUpdateBeamEndpointTimerHandle))
      (Utilities|Time|ClearandInvalidateTimerbyHandle (Variables|Transient|GetServerUpdateBeamEndpointTimerHandle))
      (bind _weapon (Utilities|GetClusterLauncherWeapon)
        (:Valid
          (if (Networking|IsServer)
            (Beam|SetSimProxyBeamNiagaraEnable _weapon false)
            (else
              (Beam|SetAutoProxyBeamNiagaraEnable _weapon false))))))))

(event Custom|ClientUpdateBeamEndpoint
  ;[7.6e-05 0.1294 0.15 0.5] UpdateBeamEndpoint
  (bind _returnvalue (Utilities|Time|SetTimerforNextTickbyEvent (delegate Custom|ClientUpdateBeamEndpoint)))
  (Variables|Transient|SetClientUpdateBeamEndpointTimerHandle _returnvalue)
  (CallFunction|UpdateBeamEndpoint))

(event Custom|UpdateBeamEndpoint
  ;[7.6e-05 0.1294 0.15 0.5] UpdateBeamEndpoint
  (bind _weapon (Utilities|GetClusterLauncherWeapon)
    (:Valid
      (bind _hit (CallFunction|PerformBeamTracing))
      (if (Networking|IsServer)
        (Beam|SetBeamInfo _weapon _hit)
        (Beam|SetSimProxyBeamNiagaraEnable _weapon (not (Class|FortWeaponComponentClusterLauncher|IsTargetingNearWall (Variables|/ClusterLauncher/Gameplay/B_TidalCrane_ClusterLauncher_Laser.B_TidalCrane_ClusterLauncher_Laser_C|GetClusterLauncherWeaponComponent _weapon))))
        (else
          (Beam|SetAutoProxyBeamInfo _weapon _hit)
          (Beam|SetAutoProxyBeamNiagaraEnable _weapon (and (FortWeapon|IsTargeting _weapon) (not (Class|FortWeaponComponentClusterLauncher|IsTargetingNearWall (Variables|/ClusterLauncher/Gameplay/B_TidalCrane_ClusterLauncher_Laser.B_TidalCrane_ClusterLauncher_Laser_C|GetClusterLauncherWeaponComponent _weapon))))))))))

(event Custom|%20ServerUpdateBeamEndpoint
  ;[7.6e-05 0.1294 0.15 0.5] UpdateBeamEndpoint
  (bind _returnvalue (Utilities|Time|SetTimerbyEvent (delegate Custom|%20ServerUpdateBeamEndpoint) (ScalableFloat|GetCachedScalableFloatValue (Variables|Default|GetBeamServerUpdateInterval))))
  (Variables|Transient|SetServerUpdateBeamEndpointTimerHandle _returnvalue)
  (CallFunction|UpdateBeamEndpoint))

(event Ability|EventOnEndAbility
  (params (bWasCancelled :type Boolean))
  ;[0.15 0.07379 0.01882 0.5] OnEndAbility: Clean up.
  (Utilities|FlowControl|Sequence
    (:then_0
      (CallFunction|SetWeaponAttackingEnabled false)
      (Utilities|Time|ClearandInvalidateTimerbyHandle (Variables|Transient|GetClientUpdateBeamEndpointTimerHandle))
      (Utilities|Time|ClearandInvalidateTimerbyHandle (Variables|Transient|GetServerUpdateBeamEndpointTimerHandle)))
    (:then_1
      (|CallParentFunction bWasCancelled))))

(fn SetWeaponAttackingToDisabled
  (CallFunction|SetWeaponAttackingEnabled false))
