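; Event-activated gameplay ability for the TidalCrane Cluster Launcher that applies the Laser AllowSecondary effect to its owner, apparently to permit the laser secondary action.
; On activation it checks that the avatar is a valid Fort player pawn before applying the effect; a failed cast applies nothing.
; The ability then ends immediately and reads nothing from the event data.
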(event Ability|EventActivateAbilityFromEvent
  (params (EventData :type Gameplay%20Event%20Data%20Structure%20%28by%20ref%29))
  (CallFunction|GetAvatarasValidFortPlayerPawn
    (:ValidFortPlayerPawn
      (Ability|ApplyGameplayEffectToOwner "/ClusterLauncher/Gameplay/GE_TidalCrane_ClusterLauncher_Laser_AllowSecondary.GE_TidalCrane_ClusterLauncher_Laser_AllowSecondary_C"))
    (:AvatarCastFailed
      _))
  (Ability|EndAbility))
