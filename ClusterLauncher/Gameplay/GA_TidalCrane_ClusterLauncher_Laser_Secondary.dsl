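; Secondary-fire ability for the Cluster Launcher laser weapon that steers fired missiles toward where the player aims while input is held.
; A 0.05-second timer sets the homing target to an aimed enemy with line of sight, else the aim point, ignoring overlapped bushes when inside one.
; Being near a wall clears homing, as does ending the ability.
; A gameplay event starts the beam visual ability and is resent if its tag is missing.
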(event Ability|EventActivateAbility
  ;[1 1 1] Ability Activate
  (bind _outfortplayerpawn (CallFunction|GetAvatarasValidFortPlayerPawn)
    (:ValidFortPlayerPawn
      (bind _output_get (Variables|Default|SetPlayer%20Pawn _outfortplayerpawn))
      (Utilities|Array|AddUnique (Variables|Default|GetBase%20Ignored%20Actors) _output_get)
      (Variables|Default|SetPlayer%20Controller (Fortnite|GetFortPlayerControllerfromActor _output_get))
      (bind _asb_tidal_crane_cluster_launcher_laser (Utilities|Casting|CastToB_TidalCrane_ClusterLauncher_Laser (Variables|/Script/FortniteGame.FortPawn|GetCurrentWeapon _output_get)))
      (Rendering|SetVisibility (Variables|/ClusterLauncher/Gameplay/B_TidalCrane_ClusterLauncher_Laser.B_TidalCrane_ClusterLauncher_Laser_C|GetNS_ClusterLauncher_Targeting _asb_tidal_crane_cluster_launcher_laser) true true)
      (bind _output_get_2 (Variables|Default|SetCurrent%20Weapon _asb_tidal_crane_cluster_launcher_laser))
      (FortWeapon|SetTargeting _output_get_2 true)
      (Utilities|Time|SetTimerbyEvent (EventDispatchers|CreateEvent :!event Custom|Update%20Homing%20Data) 0.05 true)
      (Ability|SendGameplayEvent "(TagName=\"Gameplay.Event.ClusterLaucher.BeamVisualTrace\")" (Utilities|Struct|MakeGameplayEventData "(TagName=\"Gameplay.Event.ClusterLaucher.BeamVisualTrace\")" (Variables|Default|GetPlayer%20Pawn))))))

(event Custom|Update%20Homing%20Data
  (bind _current_weapon (Variables|Default|GetCurrent%20Weapon))
  ;[1 1 1] Update Homing Data
  (if (not (Class|FortWeaponComponentClusterLauncher|IsTargetingNearWall (Variables|/ClusterLauncher/Gameplay/B_TidalCrane_ClusterLauncher_Laser.B_TidalCrane_ClusterLauncher_Laser_C|GetClusterLauncherWeaponComponent (Variables|Default|GetCurrent%20Weapon))))
    (Utilities|FlowControl|Sequence
      (:then_0
        (bind _foundenemy (Default|GetAimedAtEnemyWithLOS))
        (bind _output_get (Variables|Default|SetFound%20Enemy _foundenemy))
        (Utilities|IsValid _output_get
          (:Is%20Valid
            (Development|PrintString "Valid Target" false false)
            (Default|SetNewHomingTarget (Variables|Default|GetCurrent%20Weapon) (Utilities|Operators|Subtract (Transformation|GetActorLocation (Variables|Default|GetFound%20Enemy)) (Math|Vector|MakeVector 0.0 0.0 (ScalableFloat|GetCachedScalableFloatValue (Variables|Default|GetRow%20PawnTargetHalfHeight)))) (Camera|GetCameraRotation (Variables|/Script/Engine.PlayerController|GetPlayerCameraManager (Variables|Default|GetPlayer%20Controller)))))
          (:Is%20Not%20Valid
            (Development|PrintString "Invalid Target" false false)
            (bind _output_get_2 (Variables|Default|SetIsInBush (GameplayTags|HasMatchingGameplayTag (Variables|Default|GetPlayer%20Pawn) "(TagName=\"Gameplay.Action.Player.InBushProp\")")))
            (if _output_get_2
              (Utilities|Array|Clear (Variables|Default|GetInside%20Bush%20Ignored%20Actors))
              (for _array_element (Collision|GetOverlappingActors (Variables|Default|GetPlayer%20Pawn) "/Script/FortniteGame.BuildingProp" :!class /Script/Engine.Actor)
                (bind _returnvalue (GameplayTags|HasMatchingGameplayTag(Message) _array_element "(TagName=\"Gameplay.Prop.Bush\")"))
                (if _returnvalue
                  (Utilities|Array|Add (Variables|Default|GetInside%20Bush%20Ignored%20Actors) _array_element)))
              (Utilities|Array|Add (Variables|Default|GetInside%20Bush%20Ignored%20Actors) (Variables|Default|GetPlayer%20Pawn))
              (else
                _))
            (bind _returnvalue_2 (Fortnite|GetPlayerAimPoint (Variables|Default|GetPlayer%20Controller) (ScalableFloat|GetCachedScalableFloatValue (Variables|Default|GetRow%20MaxTargetingDist)) (select (Variables|Default|GetIsInBush) (Variables|Default|GetInside%20Bush%20Ignored%20Actors) (Variables|Default|GetBase%20Ignored%20Actors))))
            (Default|SetNewHomingTarget (Variables|Default|GetCurrent%20Weapon) _returnvalue_2 (Camera|GetCameraRotation (Variables|/Script/Engine.PlayerController|GetPlayerCameraManager (Variables|Default|GetPlayer%20Controller))))
            (Utilities|Array|Clear (Variables|Default|GetInside%20Bush%20Ignored%20Actors)))))
      (:then_1
        ;; Handle edge case of player spamming seconday input such that the visual GA does not reactivate after being canceled from removing this GA's owned tag. In other words, if we're updating homing and the beam visual is not active, activate it again here.
        (if (not (GameplayTags|HasMatchingGameplayTag (Variables|Default|GetPlayer%20Pawn) "(TagName=\"Gameplay.ClusterLaucher.Secondary.BeamVisualTracing\")"))
          (Ability|SendGameplayEvent "(TagName=\"Gameplay.Event.ClusterLaucher.BeamVisualTrace\")" (Utilities|Struct|MakeGameplayEventData "(TagName=\"Gameplay.Event.ClusterLaucher.BeamVisualTrace\")" (Variables|Default|GetPlayer%20Pawn))))))
    (elif (Variables|/ClusterLauncher/Gameplay/B_TidalCrane_ClusterLauncher_Laser.B_TidalCrane_ClusterLauncher_Laser_C|GetIsHomingMissilesToTarget _current_weapon)
      (Class|BTidalCraneClusterLauncherLaser|ClearHomingTargetforProjectiles _current_weapon)
      (Development|PrintString "Cleared Homing Because Close To Wall!" true false "(R=1.000000,G=0.976606,B=0.038114,A=1.000000)"))))

(event Ability|EventOnAbilityInputReleased
  ;[1 1 1] On Input Released
  (|CallParentFunction)
  (FortWeapon|SetTargeting (Variables|Default|GetCurrent%20Weapon) false)
  (Ability|EndAbility))

(event Ability|EventOnEndAbility
  (params (bWasCancelled :type Boolean))
  ;[1 1 1] On Ability End
  (|CallParentFunction bWasCancelled)
  (FortWeapon|SetTargeting (Variables|Default|GetCurrent%20Weapon) false)
  (Class|BTidalCraneClusterLauncherLaser|ClearHomingTargetforProjectiles (Variables|Default|GetCurrent%20Weapon)))

(fn GetAimedatEnemywithLOS
  (bind _returnvalue (Fortnite|GetAimedAtEnemy (Variables|Default|GetPlayer%20Controller) (ScalableFloat|GetCachedScalableFloatValue (Variables|Default|GetRow%20PawnLockHalfWidth)) (ScalableFloat|GetCachedScalableFloatValue (Variables|Default|GetRow%20PawnLockHalfHeight)) (ScalableFloat|GetCachedScalableFloatValue (Variables|Default|GetRow%20PawnLockRange)) true))
  (bind _output_get (LocalVariables|SetTargetEnemy _returnvalue))
  (Utilities|IsValid _output_get
    (:Is%20Valid
      (bind _returnvalue_2 (Fortnite|CheckLineOfSightToActorWithChannel (Pawn|GetPlayerViewPoint (Variables|Default|GetPlayer%20Controller)) (LocalVariables|GetTargetEnemy) "ECC_GameTraceChannel7" (Variables|Default|GetPlayer%20Pawn)))
      (if _returnvalue_2
        (return (LocalVariables|GetTargetEnemy))
        (else
          _)))
    (:Is%20Not%20Valid
      _))
  (return 0))

(fn CanActivateAbility
  (params (ActorInfo :type Gameplay%20Ability%20Actor%20Info%20Structure) (Handle :type Gameplay%20Ability%20Spec%20Handle%20Structure))
  (bind (_owneractor _avataractor) (Utilities|Struct|BreakGameplayAbilityActorInfo ActorInfo))
  (bind _asfort_player_pawn (Utilities|Casting|CastToFortPlayerPawn _avataractor)
    (:then
      (if (not (GameplayTags|HasMatchingGameplayTag _asfort_player_pawn "(TagName=\"Gameplay.ClusterLaucher.PreventSecondary\")"))
        (bind _asb_tidal_crane_cluster_launcher_laser (Utilities|Casting|CastToB_TidalCrane_ClusterLauncher_Laser (Weapon|GetCurrentWeapon _asfort_player_pawn))
          (:then
            (if (not (and (FortWeapon|IsReloading _asb_tidal_crane_cluster_launcher_laser) (Utilities|Array|IsEmpty (Variables|/ClusterLauncher/Gameplay/B_TidalCrane_ClusterLauncher_Laser.B_TidalCrane_ClusterLauncher_Laser_C|GetFired%20Projectiles _asb_tidal_crane_cluster_launcher_laser))))
              (return 0 true)))
          (:CastFailed
            _))
        (return 0 false)))
    (:CastFailed
      _)))
